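# Post-pose Anim Blueprint for the female GarageBand body cosmetic, adding secondary motion to its jacket, choker and bracelets.
# It copies the attached parent mesh's pose, then chains Anim Dynamics nodes in component space.
# Three jacket bones use linear springs with locked rotation, their alpha rising with pawn 2D speed.
# The choker is weight-constrained to head, jaw, neck, clavicle and spine bones before its simulation; choker and bracelets simulate at full alpha.
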AnimBlueprint /BRCosmetics/Characters/Player/Female/Medium/Bodies/F_MED_GarageBand/Meshes/F_MED_Garageband_AnimBp
target_skeleton=&/F_MED_Garageband_Skeleton
graphs [
 AnimGraph {nodes [
   n1 = MapRangeClamped(Value=Get PawnSpeed2D(), InRangeA=0, InRangeB=100, OutRangeA=0.1, OutRangeB=0.75)
   n2 = Anim Dynamics - Bone: dyn_jacket_l(ComponentPose=Local To Component(LocalPose=Copy Pose From Mesh() {bUseAttachedParent=true}), Alpha=@n1, ExternalForce="0,0,0") {
    bGravityOverrideInSimSpace=true
    bLinearSpring=true
    gravityScale=0.05
    linearSpringConstant=10
    numSolverIterationsPostUpdate=4
    numSolverIterationsPreUpdate=16
    alphaBoolBlend {blendOption=Linear}
    boundBone {boneName=dyn_jacket_l}
    componentLinearVelScale=<1 1 1>
    physicsBodyDefinitions [{
      sphereCollisionRadius=0
      boundBone {boneName=dyn_jacket_l}
      boxExtents=<3 3 3>
      constraintSetup {angularLimitsMax=<0 0 0>, angularLimitsMin=<0 0 0>, angularTarget=<0 0 0>, linearAxesMax=<1.5 1.5 1.5>, linearAxesMin=<-1.5 -1.5 -1.5>}
      localJointOffset=<-0.1 0 0.1>
     }]
   }
   n3 = Anim Dynamics - Bone: dyn_jacket_b(ComponentPose=@n2, Alpha=@n1, ExternalForce="0,0,0") {
    bGravityOverrideInSimSpace=true
    bLinearSpring=true
    gravityScale=0.05
    linearSpringConstant=15
    numSolverIterationsPostUpdate=4
    numSolverIterationsPreUpdate=16
    alphaBoolBlend {blendOption=Linear}
    boundBone {boneName=dyn_jacket_b}
    componentLinearVelScale=<1 1 1>
    physicsBodyDefinitions [{
      sphereCollisionRadius=0
      boundBone {boneName=dyn_jacket_b}
      boxExtents=<3 3 3>
      constraintSetup {angularLimitsMax=<0 0 0>, angularLimitsMin=<0 0 0>, angularTarget=<0 0 0>, linearAxesMax=<1.5 1.5 1.5>, linearAxesMin=<-1.5 -1.5 -1.5>}
      localJointOffset=<0 0.1 0.1>
     }]
   }
   n4 = Anim Dynamics - Bone: dyn_jacket_r(ComponentPose=@n3, Alpha=@n1, ExternalForce="0,0,0") {
    bGravityOverrideInSimSpace=true
    bLinearSpring=true
    gravityScale=0.07
    linearSpringConstant=10
    numSolverIterationsPostUpdate=4
    numSolverIterationsPreUpdate=16
    alphaBoolBlend {blendOption=Linear}
    boundBone {boneName=dyn_jacket_r}
    componentLinearVelScale=<1 1 1>
    physicsBodyDefinitions [{
      sphereCollisionRadius=0
      boundBone {boneName=dyn_jacket_r}
      boxExtents=<3 3 3>
      constraintSetup {angularLimitsMax=<0 0 0>, angularLimitsMin=<0 0 0>, angularTarget=<0 0 0>, linearAxesMax=<1.5 1.5 1.5>, linearAxesMin=<-1.5 -1.5 -1.5>}
      localJointOffset=<0.1 0 0.1>
     }]
   }
   n5 = "Constraint - dyn_chocker (PPPPPPP)"!(
    ComponentPose=@n4
    Alpha=1
    ConstraintWeights_0=0.0258
    ConstraintWeights_1:float=0.0292
    ConstraintWeights_2:float=0.4222
    ConstraintWeights_3:float=0.0172
    ConstraintWeights_4:float=0.0008
    ConstraintWeights_5:float=0.0537
    ConstraintWeights_6:float=0.4512
   ) -> (Pose) {
    alphaBoolBlend {blendOption=Linear}
    boneToModify {boneName=dyn_chocker}
    constraintSetup [
     {targetBone {boneName=head}}
     {targetBone {boneName=clavicle_l}}
     {targetBone {boneName=spine_05}}
     {targetBone {boneName=C_jaw}}
     {targetBone {boneName=clavicle_r}}
     {targetBone {boneName=neck_02}}
     {targetBone {boneName=neck_01}}
    ]
    constraintWeights [1,1,1,1,1,1,1]
   }
   n6 = Anim Dynamics - Bone: dyn_chocker(ComponentPose=@n5, Alpha=1, ExternalForce="0,0,0") {
    bGravityOverrideInSimSpace=true
    numSolverIterationsPostUpdate=4
    numSolverIterationsPreUpdate=16
    alphaBoolBlend {blendOption=Linear}
    boundBone {boneName=dyn_chocker}
    physicsBodyDefinitions [{
      sphereCollisionRadius=0
      boundBone {boneName=dyn_chocker}
      boxExtents=<5 5 5>
      constraintSetup {angularLimitsMax=<10 10 10>, angularLimitsMin=<-10 -10 -10>, angularTarget=<0 0 0>}
      localJointOffset=<0 0 -3>
     }]
   }
   n7 = Anim Dynamics - Bone: dyn_bracelets(ComponentPose=@n6, Alpha=1, ExternalForce="0,0,0") {
    bGravityOverrideInSimSpace=true
    numSolverIterationsPostUpdate=4
    numSolverIterationsPreUpdate=16
    alphaBoolBlend {blendOption=Linear}
    boundBone {boneName=dyn_bracelets}
    physicsBodyDefinitions [{
      sphereCollisionRadius=0
      boundBone {boneName=dyn_bracelets}
      boxExtents=<5 5 5>
      constraintSetup {angularLimitsMax=<10 0 0>, angularLimitsMin=<-10 0 0>, angularTarget=<0 0 0>, linearAxesMax=<3 0 0>, linearAxesMin=<-0.1 0 0>}
      localJointOffset=<0 0 0>
     }]
   }
   Output Pose(Result=Component To Local(ComponentPose=@n7)) {layerGroup=None}
  ]}
 EventGraph {nodes [
   Event BlueprintUpdateAnimation()
   Event BlueprintInitializeAnimation()
  ]}
]
variables []
